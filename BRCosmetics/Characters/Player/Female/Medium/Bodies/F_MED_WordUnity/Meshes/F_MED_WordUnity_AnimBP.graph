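# Animation Blueprint for the F_MED_WordUnity female medium player body cosmetic, posing it from the mesh it is attached to.
# Copy Pose From Mesh feeds the body's own Control Rig, whose result is the final pose.
# Each animation update, PlayerMovement fills the isMovement vector, which is passed to the rig along with a crouch flag to drive procedural adjustment on the copied pose.
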AnimBlueprint /BRCosmetics/Characters/Player/Female/Medium/Bodies/F_MED_WordUnity/Meshes/F_MED_WordUnity_AnimBP
target_skeleton=&/F_MED_WordUnity_Skeleton
graphs [
 AnimGraph {nodes [
   n1 = Control Rig!(
    Source=Copy Pose From Mesh() {bUseAttachedParent=true}
    Alpha
    isCrouching:bool=Get bIsCrouching()
    isMovement:Vector=Get isMovement() -> (isMovement:Vector)
   ) -> (Pose) {controlRigAssetReference {blueprintRigClass=&/F_MED_WordUnity_CtrlRig.F_MED_WordUnity_CtrlRig_C}}
   Output Pose(Result=@n1) {layerGroup=None}
  ]}
 EventGraph {nodes [
   Event BlueprintUpdateAnimation()
   > n1 = PlayerMovement(AnimInstance=Self-Reference(), XDotProductCoefficient=1, YDotProductCoefficient=1, ZDotProductCoefficient=1)
   > Set!(execute, isMovement_X:double=@n1, isMovement_Y:double=@n1.YMovement, isMovement_Z:double=@n1.ZMovement) -> (then, Output_Get:Vector)
  ]}
]
variables [isMovement {type=Vector}]
