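# Control Rig for the F_MED_WordUnity female character's face accessory part that turns head orientation into a curve value.
# A spherical pose reader on the head bone measures the pose within an asymmetric active region and falloff, with debug drawing enabled.
# Its output sets the SPR_HeadUp curve; the ponytail bones in the hierarchy are not driven here.
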ControlRigBlueprint /BRCosmetics/Characters/Player/Female/Medium/Bodies/F_MED_WordUnity/Meshes/Parts/F_MED_WordUnity_FaceAcc_CtrlRig
graphs [
 RigVMFunctionLibrary {links [], nodes []}
 RigVMModel {nodes [
   RigUnit_BeginExecution = &RigUnit_BeginExecution()
   n1 = &RigUnit_SphericalPoseReader(
    ActiveRegionScaleFactors=(PositiveWidth=1,NegativeWidth=0.5,PositiveHeight=0.75,NegativeHeight=0.75)
    ActiveRegionSize=0.1
    Debug=(bDrawDebug=true,bDraw2D=False,bDrawLocalAxes=False,DebugScale=25,DebugSegments=20,DebugThickness=0.25)
    DriverAxis=(X=0,Y=1,Z=0)
    DriverItem=(Type=Bone,Name="head")
    FalloffRegionScaleFactors=(PositiveWidth=1,NegativeWidth=0.5,PositiveHeight=1,NegativeHeight=1)
    FalloffSize=0.3
    OptionalParentItem=(Type=Bone,Name="None")
    RotationOffset=(X=90,Y=0,Z=-19)
    ExecutePin=@RigUnit_BeginExecution.ExecutePin
   )
   > /Script/ControlRig.RigUnit_SetCurveValue(Curve=SPR_HeadUp, Value=@n1.OutputParam) {is=struct RigVMUnitNode}
  ]}
]
hierarchy [
 root {loc=<0 0 0>, rot=<0 0 0 1>}
 pelvis {parent=root, loc=<0 -0.5301 98.0571>, rot=<0.0127 0.707 -0.0127 -0.707>}
 spine_01 {parent=pelvis, loc=<6.1208 -0.9328 0>, rot=<0 0 0.034 -0.9994>}
 spine_02 {parent=spine_01, loc=<6.5996 -0.2095 0>, rot=<0 0 0.0238 -0.9997>}
 spine_03 {parent=spine_02, loc=<6.4914 -0.1237 0>, rot=<0 0 0.0149 -0.9999>}
 spine_04 {parent=spine_03, loc=<6.9598 0.0794 0>, rot=<0 0 -0.0132 -0.9999>}
 spine_05 {parent=spine_04, loc=<6.9987 0 0>, rot=<0 0 -0.0015 -1>}
 neck_01 {parent=spine_05, loc=<11.3138 0.1529 0>, rot=<0 0 -0.2003 -0.9797>}
 neck_02 {parent=neck_01, loc=<4.5627 0 0>, rot=<0 0 0.0504 -0.9987>}
 head {parent=neck_02, loc=<5.3914 0 0>, rot=<0 0 0.115 -0.9934>}
 dyn_ponytail {parent=head, loc=<14.0346 -9.6428 -0.7097>, rot=<0.3645 -0.6059 0.6059 -0.3645>}
 dyn_ponytail_a {parent=dyn_ponytail, loc=<8.9554 0 0>, rot=<0 -0.6051 0 -0.7962>}
 dyn_ponytail_b {parent=dyn_ponytail_a, loc=<12.4716 0 0>, rot=<0 0.2011 0 -0.9796>}
 dyn_ponytail_c {parent=dyn_ponytail_b, loc=<10.185 0 0>, rot=<0 0.5459 0 -0.8378>}
 Hair_ponytail_base {parent=head, loc=<13.5671 -9.724 0>, rot=<0.6147 -0.3495 0.3495 -0.6147>}
 dyn_Hair_ponytail_01 {parent=Hair_ponytail_base, loc=<0 0 0>, rot=<0 0 0 1>}
 dyn_Hair_ponytail_02 {parent=dyn_Hair_ponytail_01, loc=<3.8568 0 0>, rot=<0.8564 0 -0.5164 0>}
 dyn_Hair_ponytail_03 {parent=dyn_Hair_ponytail_02, loc=<5.4235 0 0>, rot=<0 0.5083 0 -0.8612>}
 dyn_Hair_ponytail_04 {parent=dyn_Hair_ponytail_03, loc=<8.1285 0 0>, rot=<0 0.113 0 -0.9936>}
 dyn_Hair_ponytail_05 {parent=dyn_Hair_ponytail_04, loc=<8.8547 0 0>, rot=<0 -0.2381 0 -0.9712>}
 dyn_Hair_ponytail_06 {parent=dyn_Hair_ponytail_05, loc=<8.4053 0 0>, rot=<0.1405 -0.6865 0.143 -0.6989>}
 dyn_Hair_ponytail_07 {parent=dyn_Hair_ponytail_06, loc=<8.41 0 0>, rot=<0 0 0 1>}
 clavicle_l {parent=spine_05, loc=<5.4831 -0.1053 -3.2946>, rot=<0.0465 -0.7289 -0.009 -0.6829>}
 clavicle_r {parent=spine_05, loc=<5.4831 -0.1053 3.2946>, rot=<0.7289 0.0465 -0.6829 0.009>}
 SPR_HeadUp {type=Curve, loc=<0 0 0>, rot=<0 0 0 1>}
]
