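# Anim Blueprint for the F_MED_WordUnity face accessory mesh, adding secondary motion to its ponytail.
# It copies the attached parent's pose, runs a Control Rig, and caches the result for switchable physics branches.
# Front end uses half-weight world-space RigidBody; otherwise head-space RigidBody runs, with AnimDynamics as fallback when rigid bodies are disabled.
# Wearing backbling, detected on customization load, selects a RigidBody branch that additively bends and scales ponytail bones.
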AnimBlueprint /BRCosmetics/Characters/Player/Female/Medium/Bodies/F_MED_WordUnity/Meshes/Parts/F_MED_WordUnity_FaceAcc_AnimBP
target_skeleton=&/F_MED_WordUnity_FaceAcc_Skeleton
graphs [
 AnimGraph {nodes [
   n1 = "Transform (Modify) Bone - Bone: dyn_Hair_ponytail_03"(
    ComponentPose=Local To Component(LocalPose=Use cached pose 'PreScale Cache'())
    Alpha=1
    Translation="0,0,0"
    Rotation="15,0.000,0.000"
    Scale="1,1,1"
   ) {
    rotationMode=BMM_Additive
    rotationSpace=BCS_BoneSpace
    translationMode=BMM_Additive
    translationSpace=BCS_BoneSpace
    alphaBoolBlend {blendOption=Linear}
    boneToModify {boneName=dyn_Hair_ponytail_03}
    rotation {pitch=15}
   }
   n2 = "Transform (Modify) Bone - Bone: dyn_Hair_ponytail_03"(ComponentPose=@n1, Alpha=1, Translation="0,0,0", Rotation="-30,0,0", Scale="1,1,1") {
    rotationMode=BMM_Additive
    rotationSpace=BCS_BoneSpace
    translationMode=BMM_Additive
    translationSpace=BCS_BoneSpace
    alphaBoolBlend {blendOption=Linear}
    boneToModify {boneName=dyn_Hair_ponytail_03}
    rotation {pitch=-30}
   }
   n3 = Anim Dynamics - Bone: dyn_Hair_ponytail_03(ComponentPose=@n2, Alpha=1, ExternalForce="0,0,0") {
    angularDampingOverride=0.9999
    angularSpringConstant=1500
    bGravityOverrideInSimSpace=true
    bOverrideAngularDamping=true
    bUseGravityOverride=true
    bUseSphericalLimits=true
    alphaBoolBlend {blendOption=Linear}
    boundBone {boneName=dyn_Hair_ponytail_03}
    componentAppliedLinearAccClamp=<1100 1100 1100>
    componentLinearAccScale=<0.5 0.5 0.5>
    componentLinearVelScale=<0.5 0.5 0.5>
    gravityOverride {z=-980}
    physicsBodyDefinitions [{
      collisionType=CustomSphere
      sphereCollisionRadius=2
      boundBone {boneName=dyn_Hair_ponytail_03}
      boxExtents=<8 15 15>
      constraintSetup {angularLimitsMax=<5 30 30>, angularLimitsMin=<-5 -5 -30>, angularTarget=<0 0 1>}
      localJointOffset=<3 0 0>
     }]
    sphericalLimits [{limitRadius=35, drivingBone {boneName=spine_05}, sphereLocalOffset=<-18 15 0>}]
   }
   n4 = Anim Dynamics - Bone: dyn_Hair_ponytail_04(ComponentPose=@n3, Alpha=1, ExternalForce="0,0,0") {
    angularDampingOverride=0.9999
    angularSpringConstant=1500
    bGravityOverrideInSimSpace=true
    bOverrideAngularDamping=true
    bUseGravityOverride=true
    bUseSphericalLimits=true
    alphaBoolBlend {blendOption=Linear}
    boundBone {boneName=dyn_Hair_ponytail_04}
    componentAppliedLinearAccClamp=<1100 1100 1100>
    componentLinearAccScale=<0.5 0.5 0.5>
    componentLinearVelScale=<0.5 0.5 0.5>
    gravityOverride {z=-980}
    physicsBodyDefinitions [{
      collisionType=CustomSphere
      sphereCollisionRadius=1.5
      boundBone {boneName=dyn_Hair_ponytail_04}
      boxExtents=<8 15 15>
      constraintSetup {angularLimitsMax=<-5 30 30>, angularLimitsMin=<-5 -10 -30>, angularTarget=<0 0 1>}
      localJointOffset=<1.5 0 0>
     }]
    sphericalLimits [{limitRadius=35, drivingBone {boneName=spine_05}, sphereLocalOffset=<-18 15 0>}]
   }
   n5 = Anim Dynamics - Bone: dyn_Hair_ponytail_05(ComponentPose=@n4, Alpha=1, ExternalForce="0,0,0") {
    angularDampingOverride=0.9999
    angularSpringConstant=1500
    bGravityOverrideInSimSpace=true
    bOverrideAngularDamping=true
    bUseGravityOverride=true
    bUseSphericalLimits=true
    alphaBoolBlend {blendOption=Linear}
    boundBone {boneName=dyn_Hair_ponytail_05}
    componentAppliedLinearAccClamp=<1100 1100 1100>
    componentLinearAccScale=<0.5 0.5 0.5>
    componentLinearVelScale=<0.5 0.5 0.5>
    gravityOverride {z=-980}
    physicsBodyDefinitions [{
      collisionType=CustomSphere
      sphereCollisionRadius=1.5
      boundBone {boneName=dyn_Hair_ponytail_05}
      boxExtents=<8 15 15>
      constraintSetup {angularLimitsMax=<-5 30 30>, angularLimitsMin=<-5 -10 -30>, angularTarget=<0 0 1>}
      localJointOffset=<1.5 0 0>
     }]
    sphericalLimits [{limitRadius=35, drivingBone {boneName=spine_05}, sphereLocalOffset=<-18 15 0>}]
   }
   n6 = RigidBody!(ComponentPose=Local To Component(LocalPose=Use cached pose 'PreScale Cache'()), Alpha, OverrideWorldGravity:Vector="0,0,0", ExternalForce) -> (Pose) {
    bClampLinearTranslationLimitToRefPose=true
    bOverrideWorldGravity=true
    cachedBoundsScale=1
    overridePhysicsAsset=&/F_MED_WordUnity_FaceAcc_Physics
    simulationSpace=BaseBoneSpace
    alphaBoolBlend {blendOption=Linear}
    baseBoneRef {boneName=head}
    overrideWorldGravity {z=-980}
    simSpaceSettings {
     maxAngularAcceleration=200
     maxAngularVelocity=3
     maxLinearAcceleration=1400
     maxLinearVelocity=250
     worldAlpha=1
     externalLinearDragV=<0.7 0.4 0.4>
     externalLinearVelocity=<2 0.2 0.2>
    }
   }
   n7 = Blend Poses by bool(BlendPose_0=Component To Local(ComponentPose=@n6), BlendPose_1=Component To Local(ComponentPose=@n5), bActiveValue=Get bRigidBodyEnabled()) {blendType=Linear}
   n8 = RigidBody(ComponentPose=Local To Component(LocalPose=Use cached pose 'PreScale Cache'()), OverrideWorldGravity="0,0,0") {
    bClampLinearTranslationLimitToRefPose=true
    bOverrideWorldGravity=true
    cachedBoundsScale=1
    overridePhysicsAsset=&F_MED_WordUnity_FaceAcc_Physics
    simulationSpace=BaseBoneSpace
    alphaBoolBlend {blendOption=Linear}
    baseBoneRef {boneName=head}
    overrideWorldGravity {z=-980}
    simSpaceSettings {
     dampingAlpha=2
     maxAngularAcceleration=200
     maxAngularVelocity=3
     maxLinearAcceleration=1400
     maxLinearVelocity=250
     worldAlpha=1
     externalLinearDragV=<0.7 0.4 0.4>
     externalLinearVelocity=<2 0.2 0.2>
    }
   }
   n9 = "Transform (Modify) Bone - Bone: dyn_Hair_ponytail_01"!(ComponentPose=@n8, bAlphaBoolEnabled:bool=True, Translation="0,0,0", Rotation="30,0,0", Scale="1,1,1") -> (Pose) {
    alphaInputType=Bool
    rotationMode=BMM_Additive
    rotationSpace=BCS_BoneSpace
    scaleMode=BMM_Additive
    scaleSpace=BCS_BoneSpace
    translationMode=BMM_Additive
    translationSpace=BCS_BoneSpace
    alphaBoolBlend {blendOption=Linear}
    alphaScaleBiasClamp {bMapRange=true, inRange {max=0, min=-25}, outRange {max=0, min=1}}
    boneToModify {boneName=dyn_Hair_ponytail_01}
    rotation {pitch=30}
   }
   n10 = "Transform (Modify) Bone - Bone: dyn_Hair_ponytail_02"!(ComponentPose=@n9, bAlphaBoolEnabled:bool=True, Translation="0,0,0", Rotation="20,0,0", Scale="0.9,1,1") -> (Pose) {
    alphaInputType=Bool
    rotationMode=BMM_Additive
    rotationSpace=BCS_BoneSpace
    scaleMode=BMM_Additive
    scaleSpace=BCS_BoneSpace
    translationMode=BMM_Additive
    translationSpace=BCS_BoneSpace
    alphaBoolBlend {blendOption=Linear}
    alphaScaleBiasClamp {bMapRange=true, inRange {max=0, min=-25}, outRange {max=0, min=1}}
    boneToModify {boneName=dyn_Hair_ponytail_02}
    rotation {pitch=20}
    scale {x=0.9}
   }
   n11 = "Transform (Modify) Bone - Bone: dyn_Hair_ponytail_03"!(ComponentPose=@n10, bAlphaBoolEnabled:bool=True, Translation="0,0,0", Rotation="8,0,0", Scale="0.85,1,1") -> (Pose) {
    alphaInputType=Bool
    rotationMode=BMM_Additive
    rotationSpace=BCS_BoneSpace
    scaleMode=BMM_Additive
    scaleSpace=BCS_BoneSpace
    translationMode=BMM_Additive
    translationSpace=BCS_BoneSpace
    alphaBoolBlend {blendOption=Linear}
    alphaScaleBiasClamp {bMapRange=true, inRange {max=0, min=-25}, outRange {max=0, min=1}}
    boneToModify {boneName=dyn_Hair_ponytail_03}
    rotation {pitch=8}
    scale {x=0.85}
   }
   n12 = "Transform (Modify) Bone - Bone: dyn_Hair_ponytail_04"!(ComponentPose=@n11, bAlphaBoolEnabled:bool=True, Translation="0,0,0", Rotation="-5,0,0", Scale="0.95,1,1") -> (Pose) {
    alphaInputType=Bool
    rotationMode=BMM_Additive
    rotationSpace=BCS_BoneSpace
    scaleMode=BMM_Additive
    scaleSpace=BCS_BoneSpace
    translationMode=BMM_Additive
    translationSpace=BCS_BoneSpace
    alphaBoolBlend {blendOption=Linear}
    alphaScaleBiasClamp {bMapRange=true, inRange {max=0, min=-25}, outRange {max=0, min=1}}
    boneToModify {boneName=dyn_Hair_ponytail_04}
    rotation {pitch=-5}
    scale {x=0.95}
   }
   n13 = "Transform (Modify) Bone - Bone: dyn_Hair_ponytail_05"!(ComponentPose=@n12, bAlphaBoolEnabled:bool=True, Translation="0,0,0", Rotation="3,0,0", Scale="0.9,0.9,0.9") -> (Pose) {
    alphaInputType=Bool
    rotationMode=BMM_Additive
    rotationSpace=BCS_BoneSpace
    scaleMode=BMM_Additive
    scaleSpace=BCS_BoneSpace
    translationMode=BMM_Additive
    translationSpace=BCS_BoneSpace
    alphaBoolBlend {blendOption=Linear}
    alphaScaleBiasClamp {bMapRange=true, inRange {max=0, min=-25}, outRange {max=0, min=1}}
    boneToModify {boneName=dyn_Hair_ponytail_05}
    rotation {pitch=3}
    scale=<0.9 0.9 0.9>
   }
   n14 = Component To Local(ComponentPose=@n13)
   n15 = Get Is Wearing Backbling()
   n16 = RigidBody(ComponentPose=Local To Component(LocalPose=Use cached pose 'PreScale Cache'()), Alpha=0.5, OverrideWorldGravity="0,0,-980") {
    bClampLinearTranslationLimitToRefPose=true
    bOverrideWorldGravity=true
    cachedBoundsScale=1
    overridePhysicsAsset=&F_MED_WordUnity_FaceAcc_Physics
    simulationSpace=WorldSpace
    alphaBoolBlend {blendOption=Linear}
    baseBoneRef {boneName=head}
    simSpaceSettings {
     maxAngularAcceleration=200
     maxAngularVelocity=3
     maxLinearAcceleration=100
     maxLinearVelocity=70
     velocityScaleZ=0.1
     worldAlpha=1
     externalLinearDragV=<0.7 0.4 0.4>
     externalLinearVelocity=<2 0.2 0.2>
    }
   }
   Output Pose(Result=Blend Poses by bool(
     BlendPose_0=Component To Local(ComponentPose=@n16)
     BlendPose_1=Blend Poses by bool(BlendPose_0=@n14, BlendPose_1=@n7, bActiveValue=@n15) {blendType=Linear}
     BlendTime_0=0
     BlendTime_1=0
     bActiveValue=Get bIsInFrontEnd()
    ) {blendType=Linear}) {layerGroup=None}
   n17 = Control Rig(Source=Copy Pose From Mesh() {bUseAttachedParent=true}) {controlRigAssetReference {blueprintRigClass=&/F_MED_WordUnity_FaceAcc_CtrlRig.F_MED_WordUnity_FaceAcc_CtrlRig_C}}
   Save cached pose 'PreScale Cache'(Pose=@n17) {cacheName=PreScale Cache}
  ]}
 EventGraph {nodes [
   Delay(Duration=0.1)
   > Do Once()
   Event BlueprintUpdateAnimation()
   > n1 = Sequence()
   > n2 = Gravity Override State Function DEPRECATED(CharacterPartAnimInstance=Self-Reference(), GameTimeInSeconds=GetGameTimeInSeconds(), GravityOverrideParams=Get Gravity Override Params())
   > Set Gravity Override(Gravity Override=@n2)
   Set SPR_HeadUp(execute=@n1.then_1, SPR_HeadUp:float=GetCurveValue(CurveName=SPR_HeadUp)) -> (Output_Get:float)
   Event OnCharacterCustomizationLoaded()
   n3 = CheckForBackbling()
   > Set Is Wearing Backbling(Is Wearing Backbling=OR Boolean!(A=@n3, B=@n3.bIsBackblingACape, C:bool=@n3.bIsBackblingLong) -> (ReturnValue))
  ]}
]
variables [
 Gravity Override {type=Vector}
 Gravity Override Params {type=GravityOverrideParamsStruct}
 Is Wearing Backbling {type=boolean}
 SPR_HeadUp {type=number}
]
